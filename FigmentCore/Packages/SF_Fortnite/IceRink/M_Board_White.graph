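# M_Board_White is a Default Lit material in the IceRink folder, named as a white board surface.
# Its graph holds a single near-white constant color, with no textures, parameters or math, and the file shows no connection from it to an output.
# The material is flagged for use with Nanite.
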Material /FigmentCore/Packages/SF_Fortnite/IceRink/M_Board_White
graph {
 properties {bUsedWithNanite=true, pixelDepthOffsetMode=PDOM_Legacy, usedShadingModels=Default Lit}
 nodes [Constant3Vector() {constant=<rgba 0.935 0.935 0.935 1>}]
}
